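# Parent masked, skeletal-mesh material for Esports01 vehicle wheels, adding animated custom-color and emissive effects to the shared simple wheel parent function.
# Rim and tire red masks, blended by vertex red, confine colors driven by two rotating noise samples, with an optional second color.
# A static switch blends an LOD rim alpha texture into the opacity mask.
# The result passes through vehicle damage and de-rez transition functions.
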Material /VehicleCosmetics/Wheels/Esports01/Materials/M_WheelParent_Esports01
graph {
 properties {
  bAutomaticallySetUsageInEditor=false
  bUseMaterialAttributes=true
  bUsedWithMeshDeformer=true
  bUsedWithSkeletalMesh=true
  blendMode=BLEND_Masked
  pixelDepthOffsetMode=PDOM_Legacy
  usedShadingModels=Default Lit
 }
 nodes [
  n1 = MaterialFunctionCall() {materialFunction=&MF_WheelParent_Simple}
  n2 = GetMaterialAttributes(None=@n1.MaterialAttributes) {attributeGetTypes [#1, #2, #3]}
  n3 = NamedRerouteDeclaration(None=@n1.VertexColorR) {name=VC_R, nodeColor=<rgba 1 0 0.1641 1>}
  n4 = LinearInterpolate(
   A=NamedRerouteDeclaration(None=ComponentMask(None=@n1.RimRGBAMask) {r=true}) {name=RIM_R, nodeColor=<rgba 0.9922 1 0 1>}
   B=NamedRerouteDeclaration(None=ComponentMask(None=@n1.TireRGBAMask) {r=true}) {name=TIRE_R, nodeColor=<rgba 0.6875 0 1 1>}
   Alpha=@n3
  )
  n5 = LinearInterpolate(A=@n2.Base Color, B=NamedRerouteUsage(), Alpha=@n4)
  n6 = NamedRerouteUsage()
  n7 = StaticSwitchParameter(
   True=LinearInterpolate(
    A=@n2.Opacity Mask
    B=TextureSample() {texture=/VehicleCosmetics/Wheels/Esports01/Textures/T_ESports01_Rim_Alpha_LOD}.R
    Alpha=OneMinus(None=@n3)
   )
   False=@n2.Opacity Mask
  ) {parameterName=Alpha_LOD}
  MaterialFunctionCall(In=MaterialFunctionCall(Material=SetMaterialAttributes!(MaterialAttributes=@n2.MaterialAttributes, Base Color=@n5, Emissive Color=@n6, Opacity Mask=@n7) {attributeSetTypes [#2, #1, #3]}) {materialFunction=&MF_BaseVehicle_Damage}.Result) {materialFunction=&MF_VehicleDeRez_Transition}
  n8 = VectorParameter() {group=5_Special Effects, parameterName=Custom Color, defaultValue=<rgba 0 0.0823 0.7529 0>}
  n9 = TextureSample(UVs=Rotator(Coordinate=TextureCoordinate() {uTiling=0.2, vTiling=0.2}) {speed=0.05}) {samplerSource=SSM_Wrap_WorldGroupSettings, texture=/VehicleCosmetics/Wheels/_SharedTextures/FX/ContrastNoiseGradient_T}
  n10 = TextureSample(UVs=Rotator(Coordinate=TextureCoordinate() {uTiling=0.3, vTiling=0.3}) {speed=0.05}) {samplerSource=SSM_Wrap_WorldGroupSettings, texture=&ContrastNoiseGradient_T}
  n11 = Clamp(None=ConstantBiasScale(None=Add(A=@n9.B, B=@n10.R)) {bias=-0.15, scale=3})
  n12 = LinearInterpolate(A=VectorParameter() {group=5_Special Effects, parameterName=Custom Color 2, defaultValue=<rgba 0.5 0.5 0.5 0>}.RGB, B=@n8.RGB, Alpha=@n11)
  n13 = Add(A=OneMinus(None=Multiply(A=@n8.RGB) {constB=7})) {constA=-0.1, constB=-0.1}
  n14 = Multiply(A=@n8.RGB) {constB=0.5}
  n15 = LinearInterpolate(A=LinearInterpolate(A=Clamp(None=@n13), B=@n14, Alpha=@n8.A), B=@n14, Alpha=@n11)
  n16 = Multiply(A=@n4, B=StaticSwitchParameter(True=Clamp(None=@n12), False=Clamp(None=@n15)) {group=5_Special Effects, parameterName=2 Custom Colors})
  NamedRerouteDeclaration(None=@n16) {name=BC, nodeColor=<rgba 0 1 0.7891 1>}
  NamedRerouteDeclaration(None=@n2.Emissive Color) {name=Emissive, nodeColor=<rgba 0 0.4141 1 1>}
  NamedRerouteDeclaration(None=Multiply(A=@n16, B=ScalarParameter() {defaultValue=4, group=5_Special Effects, parameterName=Emission Intensity}) {constB=4}) {name=Rim Emissive, nodeColor=<rgba 0.5 0 1 1>}
 ]
}
